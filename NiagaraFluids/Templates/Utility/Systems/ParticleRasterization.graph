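# Niagara Fluids utility template system that pairs a GPU particle source with a Grid3D emitter, apparently to rasterize particles into a volume.
# ParticleSourceEmitter spawns curl-noise-driven sprites that shrink over life, with a velocity-based kill test.
# SetFluidSourceAttributes gives each particle a color and a radius of half its sprite size.
# The grid emitter has mesh and volume renderers; user parameters expose bounds drawing, resolution and world size.
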NiagaraSystem /NiagaraFluids/Templates/Utility/Systems/ParticleRasterization
dependencies=null
emitters {ParticleSourceEmitter {
  simTarget=GPUComputeSim
  emitterUpdateScript [
   &EmitterState(Life Cycle Mode=::NewEnumerator1, Loop Behavior=::NewEnumerator1, Loop Duration Mode=::NewEnumerator1, UseLoopDelay=false)
   &SpawnRate(SpawnRate=2000, Use Spawn Probability=false)
  ]
  particleSpawnScript [
   &InitializeParticle(
    Lifetime Max=5
    Lifetime Min=0
    Lifetime Mode=::NewEnumerator1
    Sprite Size Mode=::NewEnumerator3
    Uniform Sprite Size=20
    UsePositionOffset=false
    Write Material Random=true
   )
   &ShapeLocation(MaskSpawnGroup=false, Offset=<-200 0 -50>, Offset Mode=::NewEnumerator0, Override Seed=false, Sphere Radius=40, UseOwnerScale=true)
   &AddVelocity(Velocity=<50 0 0>)
  ]
  particleUpdateScript [
   &ParticleState()
   &CurlNoiseForce(Noise Frequency=5, Noise Strength=100, Pan Noise=true, Pan Noise Field=<0 0 2>)
   &SolveForcesAndVelocity(Clamp Velocity=false, Limit Acceleration=false)
   &ScaleSpriteSize(Scale Factor=&Vector2DFromFloat(Value=&FloatFromCurve(FloatCurve=DataInterface() {propertyValues {Curve {keys [
          {arriveTangent=-0.7681, interpMode=RCIM_Cubic, leaveTangent=-0.2483, leaveTangentWeight=0.3691, tangentMode=RCTM_Break, time=0, value=1}
          {arriveTangent=-2.4616, arriveTangentWeight=0.3336, interpMode=RCIM_Cubic, leaveTangent=-1.5991, tangentMode=RCTM_Break, time=1, value=0}
         ]}}})))
   &KillParticles(Kill Particles=&SetBoolByFloatComparison(A=&VectorLength(Vector=$Particles.Velocity), B=1, Comparison Type=::NewEnumerator6))
   /NiagaraFluids/Modules/Particles/SetFluidSourceAttributes(
    Color:LinearColor=<rgba 0.8428 0.0237 0.901 1>
    Radius:NiagaraFloat=&Multiply_Float(A=&MakeFloatFromVector2d(Vector 2D=$Particles.SpriteSize), B=0.5)
   )
  ]
  rendererClasses [&NiagaraSpriteRendererProperties]
  renderers [{rendererClass=&NiagaraSpriteRendererProperties}]
 }}
summary {
 systemName=ParticleRasterization
 emitters [
  {emitterName=ParticleSourceEmitter, simTarget=GPUComputeSim, rendererClasses [&NiagaraSpriteRendererProperties]}
  {
   emitterName=Grid3D_ParticleRasterization_Emitter
   simTarget=GPUComputeSim
   rendererClasses [&NiagaraMeshRendererProperties, /Script/Niagara.NiagaraVolumeRendererProperties]
  }
 ]
 userVariables [
  User.DrawBounds {type=NiagaraBool, defaultValue=true}
  User.ResolutionMaxAxis {type=NiagaraInt32, defaultValue=256}
  User.WorldSpaceSize {type=Vector3f, description="", defaultValue {struct=Vector3f, value=<500 500 500>}}
 ]
}
